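; Test wind tower actor that reacts with a glow, beam effect and launch sound when a player rides its vortex.
; TowerSneezeVolume overlaps are checked on authority for a pawn with upward velocity and the skydiving tag, then fire GlowNow and TimelineGlow through a DoOnce.
; GlowNow plays particles and audio only on the Remote branch.
; The timeline drives Glow on a dynamic material, then calls a once-only ResetGlow after a delay, which replays it.
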(event EventBeginPlay
  (|CallParentFunction)
  (CallFunction|DMI))

(event Collision|EventActorBeginOverlap
  (params (OtherActor :type Actor%20Object%20Reference)))

(event EventTick
  (params (DeltaSeconds :type Float%20%28single-precision%29)))

(event Custom|Glow
  (Networking|SwitchHasAuthority
    (:Authority
      (Utilities|FlowControl|DoOnce false
        (:Completed
          (CallFunction|GlowNow)
          (bind (_direction _glowalpha) (Timeline|TimelineGlow PlayFromStart)
            (:Update
              (Rendering|Material|SetScalarParameterValue (Variables|Default|GetDynamicMaterialInstanceRef) "Glow" (Math|Float|Lerp 0.0 1e+01 _glowalpha) :!class /Script/Engine.MaterialInstanceDynamic))
            (:Finished
              (Utilities|FlowControl|Delay 2.5)
              (CallFunction|ResetGlow))))))))

(event OnComponentBeginOverlap(TowerSneezeVolume)
  (params (OverlappedComponent :type Primitive%20Component%20Object%20Reference) (OtherActor :type Actor%20Object%20Reference) (OtherComp :type Primitive%20Component%20Object%20Reference) (OtherBodyIndex :type Integer) (bFromSweep :type Boolean) (SweepResult :type Hit%20Result%20Structure%20%28by%20ref%29))
  (Utilities|Casting|CastToFortPlayerControllerAthena (Fortnite|GetControllerfromActor OtherActor))
  (Networking|SwitchHasAuthority
    (:Authority
      (bind _isskydiving (CallFunction|CheckIsRidingVortex OtherActor))
      (if _isskydiving
        (Networking|SwitchHasAuthority
          (:Authority
            (Utilities|FlowControl|DoOnce false
              (:Completed
                (CallFunction|GlowNow)
                (Timeline|TimelineGlow PlayFromStart)))))))))

(event Custom|ResetGlow
  (Utilities|FlowControl|DoOnce false
    (:Completed
      (CallFunction|GlowNow)
      (Timeline|TimelineGlow PlayFromStart))))

(event Custom|GlowNow
  (Networking|SwitchHasAuthority
    (:Remote
      (Components|Activation|Activate (Variables|Default|GetP_DadBro_Mouth_Beam_Spawn) false) ; add tag and upwartd velocity check to make sure player is exiting the tunnel while going upwards
      (Audio|PlaySoundatLocation "/Game/Athena/Sounds/LaunchPad/DirectionalLauncher/Chrome_DirectionalLauncher_Launch_Cue.Chrome_DirectionalLauncher_Launch_Cue" (Transformation|GetWorldLocation (Variables|Default|GetP_DadBro_Mouth_Beam_Spawn))))))

(fn CheckIsRidingVortex
  (params (Actor :type Actor%20Object%20Reference))
  (bind _asfort_player_pawn_athena (Utilities|Casting|CastToFortPlayerPawnAthena Actor))
  (if (> (.z (Transformation|GetVelocity _asfort_player_pawn_athena)) 0)
    (if (Utilities|Array|ContainsItem (GameplayTags|BreakGameplayTagContainer (GameplayTags|GetOwnedGameplayTags _asfort_player_pawn_athena :!class /Script/GameplayTags.BlueprintGameplayTagLibrary)) "(TagName=\"MovementMode.Skydiving\")")
      (return true)
      (else
        _))
    (else
      _))
  (return false))

(fn DMI
  (bind _returnvalue (Rendering|Material|CreateDynamicMaterialInstance (Variables|Default|GetStaticMesh) 1 "/BRMapCh6/Maps/Test/Chris/Meshes/Monsters/Monster_E_Tst_Mtl.Monster_E_Tst_Mtl" "Glow" :!class /Script/Engine.PrimitiveComponent))
  (Variables|Default|SetDynamicMaterialInstanceRef _returnvalue))

(fn ConstructionScript
  (|CallParentFunction)
  (Transformation|SetRelativeRotation (Variables|Default|GetS_FX_GeyserCone) (Math|Rotator|MakeRotator 0.0) false false))
